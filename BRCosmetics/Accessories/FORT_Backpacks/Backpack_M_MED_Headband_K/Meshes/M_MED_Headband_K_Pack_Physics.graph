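# Physics asset for the M_MED_Headband_K backpack cosmetic, a small figure with two-segment ears and a two-tailed tie on its head.
# One capsule per bone covers head, spine, pelvis, limbs, ears and tie.
# Constraints chain these along the skeleton, mostly with tight twist limits and moderate swing limits, slightly asymmetric on arms and ears.
# Neck-to-spine, clavicle and tie-base joints leave swing and twist free.
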PhysicsAsset /BRCosmetics/Accessories/FORT_Backpacks/Backpack_M_MED_Headband_K/Meshes/M_MED_Headband_K_Pack_Physics
bodies {
 head [{shapeType=Capsule, length=1.6637, radius=7.579, shapeName=head_capsule, center=<3.2027 2.3178 0.0001>, rotation=<rot 0 0 90>}]
 l_clav [{shapeType=Capsule, length=0.505, radius=0.505, shapeName=l_clav_capsule, rotation=<rot 0 -90 0>}]
 l_ear_01 [{shapeType=Capsule, length=3.7751, radius=1.5417, shapeName=l_ear_02_capsule, center=<2.4732 0 0>, rotation=<rot 90 0 0>}]
 l_ear_02 [{shapeType=Capsule, length=3.7751, radius=1.5417, shapeName=l_ear_02_capsule, center=<2.4732 0 0>, rotation=<rot 90 0 0>}]
 l_foot [{shapeType=Capsule, length=2.1195, radius=1.4989, shapeName=l_foot_capsule, center=<2.9341 0 0>, rotation=<rot 90 0 0>}]
 l_knee [{shapeType=Capsule, length=2.5928, radius=1.6402, shapeName=l_knee_capsule, center=<2.0933 0 0>, rotation=<rot 90 0 0>}]
 l_lowArm [{shapeType=Capsule, length=2.5921, radius=1.4228, shapeName=l_lowArm_capsule, center=<2.001 0 0>, rotation=<rot 90 0 0>}]
 l_thigh [{shapeType=Capsule, length=2.5546, radius=1.8667, shapeName=l_thigh_capsule, center=<2.7925 0 0>, rotation=<rot 90 0 0>}]
 l_tie [{shapeType=Capsule, length=1.9418, radius=1, shapeName=l_tie_capsule, center=<1.5088 0 0>, rotation=<rot 90 0 0>}]
 l_upArm [{shapeType=Capsule, length=3.2698, radius=2.0539, shapeName=l_upArm_capsule, center=<2.369 0 0>, rotation=<rot 90 0 0>}]
 l_wrist [{shapeType=Capsule, length=2.1363, radius=1.7341, shapeName=l_wrist_capsule, center=<1.796 0 0>, rotation=<rot 90 0 0>}]
 neck_01 [{shapeType=Capsule, length=2.2175, radius=0.9886, shapeName=neck_01_capsule}]
 pelvis [{shapeType=Capsule, length=7.045, radius=6.2834, shapeName=pelvis_capsule, center=<-2 0 0>, rotation=<rot 90 0 0>}]
 r_clav [{shapeType=Capsule, length=0.505, radius=0.505, shapeName=r_clav_capsule, rotation=<rot 0 -90 0>}]
 r_ear_01 [{shapeType=Capsule, length=3.7751, radius=1.7917, shapeName=l_ear_02_capsule, center=<2.568 0 0>, rotation=<rot 90 0 0>}]
 r_ear_02 [{shapeType=Capsule, length=3.5284, radius=1.6708, shapeName=r_ear_02_capsule, center=<2.5016 0 0>, rotation=<rot 90 0 0>}]
 r_foot [{shapeType=Capsule, length=2.3695, radius=1.7489, shapeName=r_foot_capsule, center=<2.9342 0 0>, rotation=<rot 90 0 0>}]
 r_knee [{shapeType=Capsule, length=2.8427, radius=1.8903, shapeName=r_knee_capsule, center=<2.0937 0 0>, rotation=<rot 90 0 0>}]
 r_lowArm [{shapeType=Capsule, length=3.4301, radius=1.9221, shapeName=r_lowArm_capsule, center=<1.9584 0 0>, rotation=<rot 90 0 0>}]
 r_thigh [{shapeType=Capsule, length=2.8045, radius=2.1167, shapeName=r_thigh_capsule, center=<2.7929 0 0>, rotation=<rot 90 0 0>}]
 r_tie [{shapeType=Capsule, length=1.7798, radius=1, shapeName=r_tie_capsule, center=<1.5088 0 0>, rotation=<rot 90 0 0>}]
 r_upArm [{shapeType=Capsule, length=3.8092, radius=2.4787, shapeName=r_upArm_capsule, center=<2.1502 0 0>, rotation=<rot 90 0 0>}]
 r_wrist [{shapeType=Capsule, length=2.8863, radius=2.2343, shapeName=r_wrist_capsule, center=<1.7961 0 0>, rotation=<rot 90 0 0>}]
 spine_01 [{shapeType=Capsule, length=6.181, radius=4.2456, shapeName=spine_01_capsule, rotation=<rot 0 3.2607 -89.967>}]
 spine_02 [{shapeType=Capsule, length=0.5, radius=5.3636, center=<-1.1021 -1.8073 -0.0301>, rotation=<rot 0.0409 -90 0.2994>}]
 tie_base [{
   shapeType=Capsule
   length=3.4522
   radius=1.6432
   shapeName=tie_base_capsule
   center=<1.1986 -0.3002 -0.1119>
   rotation=<rot -1.0888 -90 4.4304>
  }]
}
constraints [
 {bone1Name=spine_02, bone2Name=spine_01, swing1LimitDegrees=45, swing2LimitDegrees=45, twistLimitDegrees=45}
 {
  bone1Name=neck_01
  bone2Name=spine_02
  swing1LimitDegrees=45
  swing1Motion=Free
  swing2LimitDegrees=45
  swing2Motion=Free
  twistLimitDegrees=45
  twistMotion=Free
 }
 {bone1Name=head, bone2Name=neck_01, swing1LimitDegrees=5, swing2LimitDegrees=15, twistLimitDegrees=5}
 {bone1Name=l_ear_01, bone2Name=head, swing1LimitDegrees=45, swing2LimitDegrees=45, twistLimitDegrees=2}
 {bone1Name=l_ear_02, bone2Name=l_ear_01, swing1LimitDegrees=45, swing2LimitDegrees=45, twistLimitDegrees=2}
 {bone1Name=r_ear_01, bone2Name=head, swing1LimitDegrees=30, swing2LimitDegrees=30, twistLimitDegrees=2}
 {bone1Name=r_ear_02, bone2Name=r_ear_01, swing1LimitDegrees=30, swing2LimitDegrees=30, twistLimitDegrees=2}
 {
  bone1Name=tie_base
  bone2Name=head
  swing1LimitDegrees=45
  swing1Motion=Free
  swing2LimitDegrees=45
  swing2Motion=Free
  twistLimitDegrees=45
  twistMotion=Free
 }
 {bone1Name=r_tie, bone2Name=tie_base, swing1LimitDegrees=30, swing2LimitDegrees=30, twistLimitDegrees=2}
 {bone1Name=l_tie, bone2Name=tie_base, swing1LimitDegrees=30, swing2LimitDegrees=30, twistLimitDegrees=2}
 {bone1Name=l_upArm, bone2Name=l_clav, swing1LimitDegrees=12, swing2LimitDegrees=12, twistLimitDegrees=2}
 {bone1Name=l_lowArm, bone2Name=l_upArm, swing1LimitDegrees=30, swing2LimitDegrees=30, twistLimitDegrees=2}
 {bone1Name=l_wrist, bone2Name=l_lowArm, swing1LimitDegrees=30, swing2LimitDegrees=30, twistLimitDegrees=2}
 {
  bone1Name=l_clav
  bone2Name=spine_02
  swing1LimitDegrees=45
  swing1Motion=Free
  swing2LimitDegrees=45
  swing2Motion=Free
  twistLimitDegrees=45
  twistMotion=Free
 }
 {
  bone1Name=r_clav
  bone2Name=spine_02
  swing1LimitDegrees=45
  swing1Motion=Free
  swing2LimitDegrees=45
  swing2Motion=Free
  twistLimitDegrees=45
  twistMotion=Free
 }
 {bone1Name=r_upArm, bone2Name=r_clav, swing1LimitDegrees=10, swing2LimitDegrees=10, twistLimitDegrees=2}
 {bone1Name=r_lowArm, bone2Name=r_upArm, swing1LimitDegrees=25, swing2LimitDegrees=25, twistLimitDegrees=2}
 {bone1Name=r_wrist, bone2Name=r_lowArm, swing1LimitDegrees=25, swing2LimitDegrees=25, twistLimitDegrees=2}
 {bone1Name=l_thigh, bone2Name=pelvis, swing1LimitDegrees=16, swing2LimitDegrees=16, twistLimitDegrees=2}
 {bone1Name=l_knee, bone2Name=l_thigh, swing1LimitDegrees=16, swing2LimitDegrees=16, twistLimitDegrees=2}
 {bone1Name=l_foot, bone2Name=l_knee, swing1LimitDegrees=16, swing2LimitDegrees=16, twistLimitDegrees=2}
 {bone1Name=r_thigh, bone2Name=pelvis, swing1LimitDegrees=16, swing2LimitDegrees=16, twistLimitDegrees=2}
 {bone1Name=r_knee, bone2Name=r_thigh, swing1LimitDegrees=16, swing2LimitDegrees=16, twistLimitDegrees=2}
 {bone1Name=r_foot, bone2Name=r_knee, swing1LimitDegrees=16, swing2LimitDegrees=16, twistLimitDegrees=2}
 {bone1Name=pelvis, bone2Name=spine_01, swing1LimitDegrees=12, swing2LimitDegrees=12, twistLimitDegrees=5}
]
