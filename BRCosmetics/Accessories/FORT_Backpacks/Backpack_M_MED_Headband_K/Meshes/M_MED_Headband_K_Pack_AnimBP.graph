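# Animation Blueprint for the Headband_K backpack cosmetic, adding idle motion, secondary physics and a skydiving pose.
# A random player favors Idle over Idle_Break, then a weighted multi-bone constraint ties tie_base to head, neck and lower-face bones.
# RigidBody simulation relative to the pack bone follows, then additive bone transforms offset the pack and both thighs.
# The update event sets their interpolated alpha from bIsSkyDiving.
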AnimBlueprint /BRCosmetics/Accessories/FORT_Backpacks/Backpack_M_MED_Headband_K/Meshes/M_MED_Headband_K_Pack_AnimBP
target_skeleton=&/M_MED_Headband_K_Pack_Skeleton
graphs [
 AnimGraph {nodes [
   Local Space Ref Pose()
   n1 = Random Sequence Player!() -> (Pose:Pose Link Structure) {
    is=title AnimGraphNode_RandomPlayer
    entries [
     {
      chanceToPlay=0.7
      maxLoopCount=0
      maxPlayRate=1
      minLoopCount=0
      minPlayRate=1
      sequence=/BRCosmetics/Accessories/FORT_Backpacks/Backpack_M_MED_Headband_K/Animation/Headband_K_Backpack_Idle
      blendIn {blendOption=Linear, blendTime=0.2, customCurve=null}
     }
     {
      chanceToPlay=0.3
      maxLoopCount=0
      maxPlayRate=1
      minLoopCount=0
      minPlayRate=1
      sequence=/BRCosmetics/Accessories/FORT_Backpacks/Backpack_M_MED_Headband_K/Animation/Headband_K_Backpack_Idle_Break
      blendIn {blendOption=HermiteCubic, blendTime=0.3, customCurve=null}
     }
    ]
   }
   n2 = "Constraint - tie_base (PPPPPPPPP)"!(
    ComponentPose=Local To Component(LocalPose=@n1)
    Alpha=1
    ConstraintWeights_0=0.0966
    ConstraintWeights_1:float=0.0101
    ConstraintWeights_2:float=0.0033
    ConstraintWeights_3:float=0.0284
    ConstraintWeights_4:float=0.0293
    ConstraintWeights_5:float=0.5878
    ConstraintWeights_6:float=0.0043
    ConstraintWeights_7:float=0.0244
    ConstraintWeights_8:float=0.2158
   ) -> (Pose) {
    alphaBoolBlend {blendOption=Linear}
    boneToModify {boneName=tie_base}
    constraintSetup [
     {targetBone {boneName=head}}
     {targetBone {boneName=L_lip_lower_outer}}
     {targetBone {boneName=C_lip_lower_mid}}
     {targetBone {boneName=L_lip_corner}}
     {targetBone {boneName=R_lip_corner}}
     {targetBone {boneName=C_jaw}}
     {targetBone {boneName=neck_01}}
     {targetBone {boneName=R_lip_lower_outer}}
     {targetBone {boneName=faceAttach}}
    ]
    constraintWeights [1,1,1,1,1,1,1,1,1]
   }
   n3 = RigidBody(ComponentPose=@n2) {
    bClampLinearTranslationLimitToRefPose=true
    bOverrideWorldGravity=true
    overridePhysicsAsset=&/M_MED_Headband_K_Pack_Physics
    simulationSpace=BaseBoneSpace
    alphaBoolBlend {blendOption=Linear}
    baseBoneRef {boneName=pack}
    componentAppliedLinearAccClamp=<3000 3000 3000>
    componentLinearAccScale=<0.3 0.3 0.4>
    componentLinearVelScale=<0.2 0.2 0.2>
    overrideWorldGravity {z=-980}
   }
   n4 = Get Skydive Alpha() -> (Skydive Alpha:double)
   n5 = "Transform (Modify) Bone - Bone: pack"(ComponentPose=@n3, Alpha=@n4, Translation="0,-0.25,0", Rotation="0,0,20") {
    rotationMode=BMM_Additive
    rotationSpace=BCS_BoneSpace
    translationMode=BMM_Additive
    translationSpace=BCS_BoneSpace
    alphaBoolBlend {blendOption=Linear}
    alphaScaleBiasClamp {bInterpResult=true}
    boneToModify {boneName=pack}
    rotation {roll=20}
    translation {y=-0.25}
   }
   n6 = "Transform (Modify) Bone - Bone: l_thigh"(ComponentPose=@n5, Alpha=@n4, Translation="0,0,0", Rotation="0,-45,0", Scale="1,1,1") {
    rotationMode=BMM_Additive
    rotationSpace=BCS_BoneSpace
    translationMode=BMM_Additive
    translationSpace=BCS_BoneSpace
    alphaBoolBlend {blendOption=Linear}
    alphaScaleBiasClamp {bInterpResult=true}
    boneToModify {boneName=l_thigh}
    rotation {yaw=-45}
   }
   n7 = "Transform (Modify) Bone - Bone: r_thigh"(ComponentPose=@n6, Alpha=@n4, Translation="0,0,0", Rotation="0,-45,0", Scale="1,1,1") {
    rotationMode=BMM_Additive
    rotationSpace=BCS_BoneSpace
    translationMode=BMM_Additive
    translationSpace=BCS_BoneSpace
    alphaBoolBlend {blendOption=Linear}
    alphaScaleBiasClamp {bInterpResult=true}
    boneToModify {boneName=r_thigh}
    rotation {yaw=-45}
   }
   Output Pose(Result=Component To Local(ComponentPose=@n7)) {layerGroup=None}
  ]}
 EventGraph {nodes [
   Event BlueprintUpdateAnimation()
   > Sequence()
   > n1 = Branch(Condition=Get bIsSkyDiving())
   > Set Skydive Alpha(Skydive Alpha:double=1) -> (Output_Get:double)
   Set Skydive Alpha(execute=@n1.else, Skydive Alpha=0)
   TryGetPawnOwner()
  ]}
]
variables [Skydive Alpha {type=number}]
